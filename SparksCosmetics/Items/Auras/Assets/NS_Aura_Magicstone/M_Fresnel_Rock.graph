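# Masked, lit material for Niagara mesh particles in the Magicstone aura cosmetic, shading rock with tinted diffuse, normal and specular maps.
# Diffuse tint can blend in particle color, and the normal map blends with a noise normal.
# A particle-colored Fresnel term has exponent and toggle driven by dynamic parameters.
# A third dynamic parameter erodes a smoke-texture mask via SmoothStep, optionally multiplied by diffuse alpha.
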Material /SparksCosmetics/Items/Auras/Assets/NS_Aura_Magicstone/M_Fresnel_Rock
graph {
 properties {bAutomaticallySetUsageInEditor=false, bUsedWithNiagaraMeshParticles=true, blendMode=BLEND_Masked, usedShadingModels=Default Lit}
 nodes [
  n1 = SmoothStep(
   Max=DynamicParameter() {defaultValue=<rgba 7 1 1 0>, paramNames [Fresnel Amount, "Fresnel Switch ", --, Erosion]}.Erosion
   Value=TextureSampleParameter2D() {parameterName=Erosion Texture, texture=/Game/Effects/Fort_Effects/Textures/Tile/T_Inky_Smoke_Tile}.G
  ) {constMax=0}
  n2 = TextureSampleParameter2D() {parameterName=Diffuse Texture, texture=&/T_MagicStone_Aura_D}
  LinearInterpolate(A=@n1, B=Multiply(A=@n1, B=@n2.A), Alpha=ScalarParameter() {parameterName=UseDiffuseAsMask})
  n3 = DynamicParameter() {defaultValue=<rgba 7 1 1 0>, paramNames [Fresnel Amount, "Fresnel Switch ", --, Erosion]}
  n4 = ScalarParameter() {defaultValue=0.001, parameterName=Fresnel Reflect Fraction}
  Multiply(
   A=ParticleColor().RGB
   B=LinearInterpolate(B=Fresnel(ExponentIn=@n3.Fresnel Amount, BaseReflectFractionIn=@n4), Alpha=@n3."Fresnel Switch ") {constA=1}
  )
  n5 = TextureSampleParameter2D() {
   parameterName=Normal Texture
   samplerType=SAMPLERTYPE_Normal
   texture=&/T_MagicStone_Aura_N
  }
  Reroute(None=LinearInterpolate(
    A=@n5.RGB
    B=TextureSample() {samplerType=SAMPLERTYPE_Normal, texture=/Game/Effects/Fort_Effects/BluGlow/testNormals}.RGB
    Alpha=ScalarParameter() {defaultValue=0.25, parameterName=Use Normal Noise}
   ) {constAlpha=0.25})
  n6 = ParticleColor()
  n7 = Multiply(
   A=LinearInterpolate(
    B=VectorParameter() {parameterName=DiffuseTint, defaultValue=<rgba 0 0 0 1>}.RGB
    Alpha=ScalarParameter() {defaultValue=0.8, parameterName=Diffuse Tint Strength}
   ) {constA=1}
   B=@n2.RGB
  )
  Reroute(None=LinearInterpolate(A=@n7, B=Multiply(A=@n7, B=@n6.RGB), Alpha=ScalarParameter() {parameterName=UseParticleDiffuseTint}))
  TextureSampleParameter2D() {
   parameterName=Specular Texture
   samplerType=SAMPLERTYPE_Masks
   texture=&/T_MagicStone_Aura_S
  }
 ]
}
