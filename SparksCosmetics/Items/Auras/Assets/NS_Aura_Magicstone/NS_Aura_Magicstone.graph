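# Magicstone cosmetic aura Niagara system that surrounds a character's skeletal mesh with floating glowing rocks, flames, glows, rising stars and colored dust.
# Four rocks follow a bone with bobbing, velocity-scaled interpolation, while flames, glows and stars sample other bones.
# Particle attribute readers chain rocks to swirling satellites and their ribbon trails, and also feed dust.
# IsLargeBody and IsFrontEnd adjust sizing, offsets and intensity; a Debug emitter is disabled.
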NiagaraSystem /SparksCosmetics/Items/Auras/Assets/NS_Aura_Magicstone/NS_Aura_Magicstone
dependencies {
 usedDataInterfaces [
  &NiagaraDataInterfaceArrayFloat
  &NiagaraDataInterfaceArrayInt32
  &NiagaraDataInterfaceCurve
  &NiagaraDataInterfaceArrayFloat3
  &NiagaraDataInterfaceArrayFloat2
  &NiagaraDataInterfaceColorCurve
  &NiagaraDataInterfaceParticleRead
  /Script/Niagara.NiagaraDataInterfaceSimpleCounter
  &NiagaraDataInterfaceArrayColor
  &NiagaraDataInterfaceSkeletalMesh
 ]
 usedDynamicInputs [&MakeVector]
 usedModules [
  &self:SystemSpawnScript.SystemScriptSource.SystemScriptGraph.NiagaraNodeAssignment_0.SetVariables_#1
  &SystemState
 ]
}
emitters {
 Debug {
  bEnabled=false
  emitterSpawnScript [/SetVariables(Emitter.RockReader=DataInterface() {propertyValues {EmitterBinding {bindingMode=Other, emitterName=PerpetualRocks}}}) {isSetParameters=true}]
  emitterUpdateScript [
   &EmitterState()
   &SpawnRate(SpawnRate=0.5, Use Spawn Probability=false)
   &SpawnBurst_Instantaneous(Spawn Count=4, Use Spawn Probability=false, UseLoopCountLimit=true) {enabled=false}
  ]
  particleSpawnScript [
   &InitializeParticle(
    Color Mode=::NewEnumerator0
    Lifetime=8
    Mesh Scale Mode=::NewEnumerator3
    Mesh Uniform Scale=10
    Sprite Size Mode=::NewEnumerator3
    Uniform Sprite Size=20
    UsePositionOffset=false
    Write Material Random=true
   )
   /SetVariables(Particles.Counter:NiagaraInt32=&ModuloInt(
     A=/Niagara/DynamicInputs/DataInterface/GenerateIntFromCounter(SimpleCounter:NiagaraDataInterfaceSimpleCounter=DataInterface() {propertyValues {GpuSyncMode=None, InitialValue=0}})
     B=4
    )) {isSetParameters=true}
  ]
  particleUpdateScript [
   &ParticleState()
   &SkeletalMeshLocation(
    Bone / Socket Index=7
    Bone Sampling Mode=::NewEnumerator4
    MaskSpawnGroup=false
    Orientation Sampling=::NewEnumerator1
    Position Sampling=::NewEnumerator2
    Skeletal Mesh=$System.Skeletal Mesh
   )
   &self:DebugReadParent(
    Debug Offset:Vector3f=<0 0 20>
    Parent Index:NiagaraInt32=$Particles.Counter
    Particle Attribute Reader:NiagaraDataInterfaceParticleRead=$Emitter.RockReader
   ) {module=DebugReadParent}
   &DynamicMaterialParameters(
    Index 0 Param 1=&VectorLength(Vector=$Particles.SkeletalMeshLocation.SampledVelocity)
    Index 0 Param 2=&Add_Float(A=-0.5, B=&Multiply_Float(A=$Particles.MaxParentSpeed, B=0.0167))
    Module.Index0Param0Write=true
    Module.Index0Param1Write=true
    Module.Index0Param2Write=true
    Module.Index0Param3Write=true
   )
  ]
  rendererClasses [&NiagaraSpriteRendererProperties, &NiagaraMeshRendererProperties]
  renderers [
   {rendererClass=&NiagaraSpriteRendererProperties}
   {rendererClass=&NiagaraMeshRendererProperties, rendererIndex=1}
  ]
 }
 Dust {
  emitterSpawnScript [/SetVariables(Emitter.RockReader=DataInterface() {propertyValues {EmitterBinding {bindingMode=Other, emitterName=PerpetualRocks}}}) {isSetParameters=true}]
  emitterUpdateScript [
   &EmitterState()
   &SpawnRate(SpawnRate=18, Use Spawn Probability=false)
  ]
  particleSpawnScript [
   &PartitionParticles(Particles Per Partition=4, Partition Mode:ENiagara_PartitionMode=::NewEnumerator1)
   &InitializeParticle(
    Color=&Multiply_LinearColorByFloat(
     Linear Color=&SelectLinearColorFromArray(
      Array Sampling Mode=::NewEnumerator0
      Direct Array Index=?
      Color Selection Array=DataInterface() {propertyValues {ColorData [<rgba 2.9663 0.2348 3 1>, <rgba 3 1.8589 0.1096 1>, <rgba 0.1096 3 1.092 1>, <rgba 0.1304 1.1075 3 1>]}}
      Random Seed=0
     )
     Float=&FNTimeOfDayScalar(
      Night Scale=1
      Morning Scale=1.5
      Day Scale=&MakeCustomFloatFromBool(Bool=$User.IsFrontEnd, True Float=1, False Float=1)
      Evening Scale=1
     )
    )
    Lifetime Max=1
    Lifetime Mode=::NewEnumerator1
    Sprite Rotation Mode=::NewEnumerator2
    Sprite Size Mode=::NewEnumerator3
    Uniform Sprite Size=&Multiply_Float(
     A=25
     B=&Multiply_Float(
      A=&Lerp_Float(A=1, B=1, Alpha=$Particles.PartitionParticles.Normalized Index In Partition)
      B=&MakeCustomFloatFromBool(Bool=$User.IsLargeBody, True Float=1.3, False Float=1)
     )
    )
    UsePositionOffset=false
    Write Material Random=true
   )
   /SetVariables(Particles.VisibilityTag=&MakeCustomIntFromBool(Bool=$User.IsFrontEnd, True INT=1, False INT=0)) {isSetParameters=true}
   &DynamicMaterialParameters(
    Index 0 Param 1=0.5
    Module.Param0WriteEnabled=true
    Module.Param1WriteEnabled=false
    Module.Param2WriteEnabled=false
    Module.Param3WriteEnabled=false
   )
   /SetVariables(Particles.bIsRockCloud:NiagaraBool=&RandomBool()) {isSetParameters=true}
  ]
  particleUpdateScript [
   &ParticleState()
   &SpriteRotationRate(Rotation Rate=30)
   &SkeletalMeshLocation(
    Bone / Socket Index=&SelectIntFromArray(Direct Array Index=$Particles.PartitionParticles.Partition, Int Selection Array=DataInterface() {propertyValues {IntData [4,2]}})
    Bone Sampling Mode=::NewEnumerator4
    MaskSpawnGroup=false
    Orientation Sampling=::NewEnumerator1
    Position Sampling=::NewEnumerator1
    Sampled Position Offset=&Lerp_Vector(
     A=&SelectVectorFromArray(
      Direct Array Index=$Particles.PartitionParticles.Partition
      Vector Selection Array=DataInterface() {propertyValues {FloatData [<-5 0 3>, <5 0 -3>]}}
     )
     B=&SelectVectorFromArray(
      Direct Array Index=$Particles.PartitionParticles.Partition
      Vector Selection Array=DataInterface() {propertyValues {FloatData [<27 0 3>, <-27 0 -3>]}}
     )
     Alpha=$Particles.PartitionParticles.Normalized Index In Partition
    )
    Skeletal Mesh=$System.Skeletal Mesh
   )
   &self:ReadParent(
    Parent Index:NiagaraInt32=&RandomRangeInt(Minimum=0, Maximum=3)
    Particle Attribute Reader:NiagaraDataInterfaceParticleRead=$Emitter.RockReader
   ) {module=ReadParent}
   &ScaleColor(
    Scale Alpha=&FloatFromCurve(
     FloatCurve=DataInterface() {propertyValues {Curve {keys [
         {interpMode=RCIM_Cubic, leaveTangent=2.1138, tangentMode=RCTM_Break, time=0, value=0}
         <at 0.4 1 RCIM_Cubic>
         {arriveTangent=-3.4143, interpMode=RCIM_Cubic, leaveTangent=-3.4143, tangentMode=RCTM_User, time=1, value=0}
        ]}}}
     Scale Curve=&MakeCustomFloatFromBool(Bool=$User.IsFrontEnd, True Float=0.5, False Float=1.33)
    )
    ScaleA=true
    ScaleRGB=true
   ) {module=ScaleColor003}
   &GravityForce(Gravity=&RandomRangeVector(Minimum=<0 0 15>, Maximum=<0 1 25>))
   &SolveForcesAndVelocity(Clamp Velocity=false, Limit Acceleration=false)
   /SetVariables(Particles.Position=&MakeCustomVectorFromBool(
     Bool=$Particles.bIsRockCloud
     True Vector=&TransformVector(
      Vector=&Add_Vector(
       A=Linked() {linkedVariable {name=Particles.ParentPosition, type=NiagaraPosition}}
       B=&Multiply_VectorByFloat(Vector=Linked() {linkedVariable {name=Engine.Owner.Position, type=NiagaraPosition}}, Float=-1)
      )
      Source Space=::World
      Destination Space=::Local
     )
     False Vector=Linked() {linkedVariable {name=Particles.Position, type=NiagaraPosition}}
    )) {isSetParameters=true}
   /SetVariables(Particles.Color=&MakeLinearColorFromVectorAndFloat(
     "Vector (RGB)"=&MakeVectorFromLinearColorRGB(Color=&LerpLinearColors(StartColor=$Particles.Initial.Color, EndColor=<rgba 0.2434 0.0861 1.5 1>, LerpFactor=$Particles.NormalizedAge))
     "Float (Alpha)"=&FloatFromCurve(
      FloatCurve=DataInterface() {propertyValues {Curve {keys [
          {interpMode=RCIM_Cubic, leaveTangent=2.1138, tangentMode=RCTM_Break, time=0, value=0}
          <at 0.4 1 RCIM_Cubic>
          {arriveTangent=-3.4143, interpMode=RCIM_Cubic, leaveTangent=-3.4143, tangentMode=RCTM_User, time=1, value=0}
         ]}}}
      Scale Curve=&MakeCustomFloatFromBool(Bool=$User.IsFrontEnd, True Float=0.5, False Float=1.33)
     )
    )) {isSetParameters=true}
  ]
  rendererClasses [&NiagaraSpriteRendererProperties]
  renderers [
   {rendererClass=&NiagaraSpriteRendererProperties}
   {rendererClass=&NiagaraSpriteRendererProperties, rendererIndex=1}
  ]
 }
 Flames {
  emitterUpdateScript [
   &EmitterState()
   &SpawnBurst_Instantaneous(Spawn Count=7, Use Spawn Probability=false, UseLoopCountLimit=true)
  ]
  particleSpawnScript [
   &PartitionParticles(Number Of Partitions=7)
   &InitializeParticle(
    Color=<rgba 0.0612 0 2 0.75>
    Lifetime=0.1
    Sprite Rotation Angle=-90
    Sprite Rotation Mode=::NewEnumerator0
    Sprite Size=&SelectVector2DFromArray(
     Direct Array Index=$Particles.PartitionParticles.Partition
     Vector 2D Selection Array=DataInterface() {propertyValues {FloatData [<40 20>, <53 25>, <65 40>, <70 70>, <65 40>, <53 25>, <40 20>]}}
    )
    Sprite Size Mode=::NewEnumerator4
    UsePositionOffset=false
    Write Material Random=true
   ) {module=InitializeParticle001}
   &CameraOffset(Camera Offset Amount=-20)
  ]
  particleUpdateScript [
   &ParticleState(Kill Particles When Lifetime Has Elapsed=false, Let Infinitely Lived Particles Die When Emitter Deactivates=true)
   &SkeletalMeshLocation(
    Bone / Socket Index=&SelectIntFromArray(Direct Array Index=$Particles.PartitionParticles.Partition, Int Selection Array=DataInterface() {propertyValues {IntData [1,1,2,3,4,5,5]}})
    Bone Sampling Mode=::NewEnumerator4
    MaskSpawnGroup=false
    Orientation Sampling=::NewEnumerator1
    Position Sampling=::NewEnumerator2
    Sampled Position Offset=&SelectVectorFromArray(
     Direct Array Index=$Particles.PartitionParticles.Partition
     Vector Selection Array=DataInterface() {propertyValues {FloatData [<0 0 0>, <0 0 0>, <0 0 0>, <0 0 0>, <0 0 0>, <0 0 0>, <0 0 0>]}}
    )
    Sampled Position Offset Coordinate Space:ENiagaraSkelMeshTransforms=::NewEnumerator0
    Skeletal Mesh=$System.Skeletal Mesh
   ) {module=SkeletalMeshLocation003}
   &SkeletalMeshLocation(
    Bone / Socket Index=&SelectIntFromArray(Direct Array Index=$Particles.PartitionParticles.Partition, Int Selection Array=DataInterface() {propertyValues {IntData [1,2,2,3,4,4,5]}})
    Bone Sampling Mode=::NewEnumerator4
    MaskSpawnGroup=false
    Orientation Sampling=::NewEnumerator1
    Position Sampling=::NewEnumerator2
    Sampled Position Offset=&SelectVectorFromArray(
     Direct Array Index=$Particles.PartitionParticles.Partition
     Vector Selection Array=DataInterface() {propertyValues {FloatData [<0 0 0>, <0 0 0>, <0 0 0>, <0 0 0>, <0 0 0>, <0 0 0>, <0 0 0>]}}
    )
    Sampled Position Offset Coordinate Space:ENiagaraSkelMeshTransforms=::NewEnumerator0
    Skeletal Mesh=$System.Skeletal Mesh
   )
   &DynamicMaterialParameters(
    Index 0 Param 1=&RandomRangeFloat(Minimum=0, Maximum=1)
    Index 0 Param 2=1
    Index 0 Param 3=1
    Index 0 Param 4=0.08
    Module.Param0WriteEnabled=true
    Module.Param1WriteEnabled=true
    Module.Param2WriteEnabled=true
    Module.Param3WriteEnabled=true
   )
   &ScaleColor(
    Scale Alpha=&MakeCustomFloatFromBool(Bool=$User.IsFrontEnd, True Float=0.5, False Float=1)
    ScaleA=true
    ScaleRGB=false
   )
   &Color(Color=&ColorFromCurve(
     ColorCurve=DataInterface() {propertyValues {
       AlphaCurve {keys [<at 0 1 RCIM_Cubic>]}
       BlueCurve {keys [<at 0 0 RCIM_Cubic>,<at 0.5 0>,<at 1 1 RCIM_Cubic>]}
       GreenCurve {keys [<at 0 0 RCIM_Cubic>,<at 0.5 1>,<at 1 0 RCIM_Cubic>]}
       RedCurve {keys [<at 0 1 RCIM_Cubic>,<at 0.5 0>,<at 1 0 RCIM_Cubic>]}
      }}
     CurveIndex=&Multiply_FloatByInt(Float=0.1429, Integer=$Particles.PartitionParticles.Partition)
    )) {enabled=false}
   /SetVariables(Particles.Position=&LerpPosition(A=$Particles.SkeletalMeshLocation.SampledPosition, B=$Particles.SkeletalMeshLocation003.SampledPosition, Alpha=0.55)) {isSetParameters=true}
  ]
  rendererClasses [&NiagaraSpriteRendererProperties]
  renderers [
   {rendererClass=&NiagaraSpriteRendererProperties}
   {rendererClass=&NiagaraSpriteRendererProperties, rendererIndex=1}
  ]
 }
 GlowDecal {
  emitterUpdateScript [
   &EmitterState()
   &SpawnBurst_Instantaneous(Spawn Count=7, Use Spawn Probability=false, UseLoopCountLimit=true)
  ]
  particleSpawnScript [
   &PartitionParticles(Number Of Partitions=7)
   &InitializeParticle(
    Color=&MakeCustomLinearColorFromBool(Bool=$User.IsFrontEnd, True Color=<rgba 3.4273 0 20 1>, False Color=<rgba 11.8554 0 100 1>)
    Lifetime=0.1
    Sprite Rotation Mode=::NewEnumerator0
    Sprite Size Mode=::NewEnumerator3
    Uniform Sprite Size=&Lerp_Float(
     A=&SelectFloatFromArray(
      Direct Array Index=$Particles.PartitionParticles.Partition
      Float Selection Array=DataInterface() {propertyValues {FloatData [40,40,20,20,30,30,30]}}
     )
     B=&SelectFloatFromArray(
      Direct Array Index=&SpawnGroup()
      Float Selection Array=DataInterface() {propertyValues {FloatData [80,80,80,40,40,60,60]}}
     )
     Alpha=&MakeCustomFloatFromBool(Bool=$User.IsLargeBody, True Float=1, False Float=0)
    )
    UsePositionOffset=false
    Write Material Random=true
   ) {module=InitializeParticle001}
   &CameraOffset(Camera Offset Amount=&SelectFloatFromArray(
     Direct Array Index=&SpawnGroup()
     Float Selection Array=DataInterface() {propertyValues {FloatData [5,5,5,5,5,5]}}
    ))
  ]
  particleUpdateScript [
   &ParticleState(Kill Particles When Lifetime Has Elapsed=false, Let Infinitely Lived Particles Die When Emitter Deactivates=true)
   &SkeletalMeshLocation(
    Bone / Socket Index=&SelectIntFromArray(Direct Array Index=$Particles.PartitionParticles.Partition, Int Selection Array=DataInterface() {propertyValues {IntData [2,4,0,6,1,5,3]}})
    Bone Sampling Mode=::NewEnumerator4
    MaskSpawnGroup=false
    Orientation Sampling=::NewEnumerator1
    Sampled Position Offset Coordinate Space:ENiagaraSkelMeshTransforms=::NewEnumerator1
    Skeletal Mesh=$System.Skeletal Mesh
   ) {module=SkeletalMeshLocation003}
   &DynamicMaterialParameters(
    Index 0 Param 1=&SelectFloatFromArray(Direct Array Index=$Particles.SpawnGroup, Float Selection Array=DataInterface() {propertyValues {FloatData [5,5,5,5,8,8]}})
    Index 0 Param 3=1
    Module.Param0WriteEnabled=true
    Module.Param1WriteEnabled=true
    Module.Param2WriteEnabled=true
    Module.Param3WriteEnabled=true
   )
  ]
  rendererClasses [&NiagaraSpriteRendererProperties]
  renderers [
   {rendererClass=&NiagaraSpriteRendererProperties}
   {rendererClass=&NiagaraSpriteRendererProperties, rendererIndex=1}
  ]
 }
 PerpetualRocks {
  emitterUpdateScript [
   &EmitterState()
   &SpawnBurst_Instantaneous(Spawn Count=4, Use Spawn Probability=false, UseLoopCountLimit=true)
  ]
  particleSpawnScript [
   /SetVariables(Particles.Variety:NiagaraFloat=&RandomRangeFloat(Minimum=0, Maximum=1)) {isSetParameters=true}
   &InitializeParticle(
    Color=<rgba 0 0.4111 50 1>
    Lifetime=&RandomRangeFloat(Minimum=4, Maximum=7)
    Mesh Scale Mode=::NewEnumerator3
    Mesh Uniform Scale=&Multiply_Float(
     A=&MakeCustomFloatFromBool(Bool=$User.IsLargeBody, True Float=1.2, False Float=1)
     B=&Lerp_Float(A=1, B=1.6, Alpha=$Particles.Variety)
    )
    Sprite Rotation Mode=::NewEnumerator2
    Sprite Size Mode=::NewEnumerator3
    Uniform Sprite Size=&Lerp_Float(
     A=30
     B=40
     Alpha=&Add_Float(A=$Particles.Variety, B=&RandomRangeFloat(Minimum=-0.1, Maximum=0.2))
    )
    UsePositionOffset=false
    Write Material Random=true
   )
   &InitialMeshOrientation(
    Mesh Orientation Mode=::NewEnumerator1
    Rotation=&RandomRangeVector(Minimum=<-1 -1 -1>, Maximum=<1 1 1>)
    Use Rotation Vector=true
   )
   /SetVariables(Particles.MeshIndex=&RandomRangeInt(Minimum=0, Maximum=2)) {isSetParameters=true}
   /SetVariables(Particles.BackerCamera=-5, Particles.BackerColor:LinearColor=<rgba 0.0191 0 0.191 1>) {isSetParameters=true}
   /SetVariables(
    Particles.InterpSpeed:NiagaraFloat=&RandomRangeFloat(Minimum=5, Maximum=10)
    Particles.OffsetPosition:Vector3f=&RandomVector(Vector Scale=10, Evaluation Type=::NewEnumerator1)
   ) {isSetParameters=true}
  ]
  particleUpdateScript [
   &SolveForcesAndVelocity(Clamp Velocity=false, Limit Acceleration=false)
   &ParticleState(Kill Particles When Lifetime Has Elapsed=false, Let Infinitely Lived Particles Die When Emitter Deactivates=true)
   &SpriteRotationRate(Rotation Rate=&Multiply_Float(
     A=&RandomRangeFloat(Minimum=25, Maximum=50)
     B=&UniformAOrBFloat(A=-1, Evaluation Type=::NewEnumerator0, Recalculate Random Each Loop=true, Override Seed=false)
    ))
   &ScaleMeshSize(Scale Factor=&VectorFromFloat(Value=&FloatFromCurve(
      FloatCurve=DataInterface() {propertyValues {Curve {keys [
          {interpMode=RCIM_Cubic, tangentMode=RCTM_Break, time=0, value=0}
          {arriveTangent=3.6134, interpMode=RCIM_Cubic, tangentMode=RCTM_Break, time=0.5, value=1}
         ]}}}
      CurveIndex=$Emitter.Age
     )))
   &DynamicMaterialParameters(
    Index 0 Param 1=&FloatFromCurve(
     FloatCurve=DataInterface() {propertyValues {Curve {keys [
         {interpMode=RCIM_Cubic, tangentMode=RCTM_Break, time=0, value=0}
         {arriveTangent=21.842, interpMode=RCIM_Cubic, tangentMode=RCTM_Break, time=1, value=10}
        ]}}}
     CurveIndex=$Emitter.Age
    )
    Index 0 Param 2=1
    Index 0 Param 3=1
    Index 0 Param 4=&FloatFromCurve(
     FloatCurve=DataInterface() {propertyValues {Curve {keys [
         {interpMode=RCIM_Cubic, tangentMode=RCTM_Break, time=0, value=1}
         {arriveTangent=-3.5426, interpMode=RCIM_Cubic, tangentMode=RCTM_Break, time=1, value=0}
        ]}}}
     CurveIndex=$Emitter.Age
    )
    Index 1 Param 1=2
    Index 1 Param 2=0.05
    Module.Index0Param0Write=true
    Module.Index0Param1Write=true
    Module.Index0Param2Write=true
    Module.Index0Param3Write=false
    Module.Index1Param0Write=true
    Module.Index1Param1Write=true
    Module.Index1Param2Write=true
    Module.Index1Param3Write=true
    Write Parameter Index 1=true
   )
   /SetVariables(
    Particles.OffsetPosition:Vector3f=/Niagara/DynamicInputs/Lerp/SmoothLerpOverTime_Vector(
     Convergence Rate:NiagaraFloat=0.5
     Smooth Value:Vector3f=$Particles.OffsetPosition
     Target Value:Vector3f=&RandomVector(
      Vector Scale=10
      Evaluation Type=::NewEnumerator1
      Randomness Mode=::NewEnumerator1
      Random Seed=&Add_Integer(
       A=&ReturnExecIndex()
       B=&MakeIntFromFloat(Float=&Add_Float(A=$Engine.Time, B=&RandomRangeFloat(Minimum=0, Maximum=1)))
      )
      Fixed Random Seed=true
     )
    )
    Transient.BobOffset:Vector3f=&Multiply_VectorByFloat(
     Vector=<0 0 2>
     Float=&Waveform(
      Global Phase=&Add_Float(
       A=&Multiply_Float(A=$Emitter.Age, B=0.5)
       B=&ReturnNormalizedExecIndex(Normalized Index Scale=9, Normalized Index Includes One=false)
      )
      "Waveform [1]"=::NewEnumerator0
      "[1] Frequency"=&MakeCustomFloatFromBool(Bool=$User.IsFrontEnd, True Float=0.3, False Float=0.5)
      "[1] Amplitude Min/Max"=&MakeCustomVector2DFromBool(Bool=$User.IsFrontEnd, True Vector2D=<0 1>, False Vector2D=<0 2>)
     )
    )
   ) {isSetParameters=true}
   /SetVariables(
    Transient.LargeBodyOffset:Vector3f=&SelectVectorFromArray(
     Direct Array Index=&ReturnExecIndex()
     Vector Selection Array=DataInterface() {propertyValues {FloatData [<10 -10 50>, <40 -30 -20>, <40 -30 20>, <10 -10 -50>]}}
    )
    Transient.MedBodyOffset:Vector3f=&Add_Vector(
     A=<15 -18 0>
     B=&SelectVectorFromArray(
      Direct Array Index=&ReturnExecIndex()
      Vector Selection Array=DataInterface() {propertyValues {FloatData [<0 0 35>, <30 -10 -20>, <30 -10 20>, <0 0 -35>]}}
     )
    )
   ) {isSetParameters=true}
   &SkeletalMeshLocation(
    Bone / Socket Index=7
    Bone Sampling Mode=::NewEnumerator4
    MaskSpawnGroup=false
    Orientation Sampling=::NewEnumerator1
    Position Sampling=::NewEnumerator2
    Sampled Position Offset=&MakeCustomVectorFromBool(Bool=$User.IsLargeBody, True Vector=$Transient.LargeBodyOffset, False Vector=$Transient.MedBodyOffset)
    Skeletal Mesh=$System.Skeletal Mesh
   )
   /SetVariables(Transient.ScaleByVelocity:NiagaraFloat=&ClampFloat(Float=&Add_Float(
      A=-0.25
      B=&Multiply_Float(A=0.0025, B=&VectorLength(Vector=$Particles.SkeletalMeshLocation.SampledVelocity))
     ))) {isSetParameters=true}
   &InterpolateOverTime(
    Rate Of Change:NiagaraFloat=&Multiply_Float(A=$Particles.InterpSpeed, B=&Lerp_Float(A=1, B=2, Alpha=$Transient.ScaleByVelocity))
    Target Position Value:NiagaraPosition=&AddVectorToPosition(
     Position=$Particles.SkeletalMeshLocation.SampledPosition
     Vector=&Add_Vector(A=$Particles.OffsetPosition, B=$Transient.BobOffset)
    )
    Variable Type=::NewEnumerator6
   )
   /SetVariables(Particles.Position=$Particles.InterpolateOverTime.InterpolateOverTime_Position.Moving Average) {isSetParameters=true}
   &UpdateMeshOrientation(
    Rotation Rate=&RandomRangeFloat(Minimum=0.2, Maximum=0.4)
    Rotation Vector=&RandomVector(Vector Scale=1)
   )
   &CalculateAccurateVelocity()
  ]
  rendererClasses [&NiagaraMeshRendererProperties, &NiagaraSpriteRendererProperties]
  renderers [
   {rendererClass=&NiagaraMeshRendererProperties, rendererIndex=0}
   {rendererClass=&NiagaraSpriteRendererProperties, rendererIndex=1}
  ]
 }
 RockSatellites {
  emitterSpawnScript [/SetVariables(Emitter.RockReader=DataInterface() {propertyValues {EmitterBinding {bindingMode=Other, emitterName=PerpetualRocks}}}) {isSetParameters=true}]
  emitterUpdateScript [
   &EmitterState()
   &SpawnBurst_Instantaneous(Spawn Count=8, Use Spawn Probability=false, UseLoopCountLimit=true)
   &SpawnRate(SpawnRate=3.5, Use Spawn Probability=false) {enabled=false}
  ]
  particleSpawnScript [
   &PartitionParticles(Number Of Partitions=4, Write Partition to Ribbon ID=::NewEnumerator0)
   &self:ReadParent(
    Parent Index:NiagaraInt32=$Particles.PartitionParticles.Partition
    Particle Attribute Reader:NiagaraDataInterfaceParticleRead=$Emitter.RockReader
   ) {module=ReadParent}
   &InitializeParticle(
    AdjustAlpha=false
    AdjustHue=false
    AdjustSaturation=true
    AdjustValue=true
    Color=<rgba 0 0.4111 50 1>
    Color Mode=::NewEnumerator3
    Lifetime=2
    Mesh Scale Mode=::NewEnumerator3
    Mesh Uniform Scale=&Multiply_Float(
     A=&RandomRangeFloat(Minimum=0.25, Maximum=0.35)
     B=&MakeCustomFloatFromBool(Bool=$User.IsLargeBody, True Float=1.5, False Float=1)
    )
    Position Mode=::NewEnumerator1
    Position Offset=&RandomVector(Vector Scale=&MakeCustomFloatFromBool(Bool=$User.IsLargeBody, True Float=10, False Float=7))
    Saturation Range:Vector2f=<0.5 1>
    Sprite Size Mode=::NewEnumerator3
    UsePositionOffset=true
    Value Range:Vector2f=<0.5 1>
    Write Material Random=true
   )
   /SetVariables(Particles.MeshIndex=&RandomRangeInt(Minimum=0, Maximum=2)) {isSetParameters=true}
   /SetVariables(Particles.BackerCamera=-5, Particles.BackerColor:LinearColor=<rgba 0.0191 0 0.191 0.6>) {isSetParameters=true}
  ]
  particleUpdateScript [
   &ParticleState(Kill Particles When Lifetime Has Elapsed=false)
   &ScaleMeshSize(Scale Factor=&VectorFromFloat(Value=&FloatFromCurve(
      FloatCurve=DataInterface() {propertyValues {Curve {keys [
          <at 0 0 RCIM_Cubic>
          {arriveTangentWeight=0.1491, interpMode=RCIM_Cubic, leaveTangentWeight=0.2236, tangentMode=RCTM_User, time=0.4, value=1}
         ]}}}
      CurveIndex=$Particles.Age
     )))
   &self:ReadParentPos+Vel(Parent ID:NiagaraID=$Particles.ParentID, Particle Attribute Reader:NiagaraDataInterfaceParticleRead=$Emitter.RockReader) {module=ReadParentPos+Vel}
   &InterpolateOverTime(
    Target Vector Value:Vector3f=&RandomVector(Vector Scale=1, Evaluation Type=::NewEnumerator1)
    Variable Type=::NewEnumerator2
   )
   &VortexForce(
    Origin Pull Only On Vortex Axis=false
    UseOriginOffset=false
    Vortex Axis=$Particles.InterpolateOverTime.InterpolateOverTime_V3_.Moving Average
    Vortex Force Amount=10
    Vortex Origin:NiagaraPosition=<0 0 0>
   )
   &AccelerationForce(Acceleration=&Multiply_VectorByFloat(Vector=$Particles.ParentVelocity, Float=-0.1))
   &SolveForcesAndVelocity(Clamp Velocity=false, Limit Acceleration=false)
   &DynamicMaterialParameters(
    Index 0 Param 1=&FloatFromCurve(
     FloatCurve=DataInterface() {propertyValues {Curve {keys [
         {interpMode=RCIM_Cubic, tangentMode=RCTM_Break, time=0, value=0}
         {arriveTangent=11.1211, interpMode=RCIM_Cubic, tangentMode=RCTM_Break, time=1, value=5}
        ]}}}
     CurveIndex=$Emitter.Age
    )
    Index 0 Param 2=1
    Index 0 Param 4=&FloatFromCurve(
     FloatCurve=DataInterface() {propertyValues {Curve {keys [
         {interpMode=RCIM_Cubic, tangentMode=RCTM_Break, time=0, value=1}
         {arriveTangent=-3.5426, interpMode=RCIM_Cubic, tangentMode=RCTM_Break, time=1, value=0}
        ]}}}
     CurveIndex=$Emitter.Age
    )
    Index 1 Param 1=2
    Index 1 Param 2=0.05
    Module.Index0Param0Write=true
    Module.Index0Param1Write=true
    Module.Index0Param2Write=true
    Module.Index0Param3Write=false
    Module.Index1Param0Write=true
    Module.Index1Param1Write=true
    Module.Index1Param2Write=true
    Module.Index1Param3Write=true
    Write Parameter Index 1=true
   )
   &Color(Color=&ColorFromCurve(
     ColorCurve=DataInterface() {propertyValues {
       AlphaCurve {keys [<at 0 1 RCIM_Cubic>]}
       BlueCurve {keys [<at 0 0 RCIM_Cubic>, {arriveTangent=1.0032, interpMode=RCIM_Cubic, leaveTangent=1.0032, time=0.5, value=0}, <at 0.9969 1>]}
       GreenCurve {keys [<at 0 0 RCIM_Cubic>,<at 0.5 1 RCIM_Cubic>,<at 0.9969 0>]}
       RedCurve {keys [<at 0 1 RCIM_Cubic>, {arriveTangent=-1.0032, interpMode=RCIM_Cubic, leaveTangent=-1.0032, time=0.5, value=0}, <at 0.9969 0>]}
      }}
     CurveIndex=&Multiply_Float(
      A=&RemapRange(
       Input Value To Remap=&DistanceBetweenPositions(A=$Particles.Position, B=$Particles.ParentPosition)
       Input Min=12
       Input Max=30
       Output Min=0
       Output Max=1
      )
      B=1
     )
    )) {enabled=false}
   /SetVariables(Particles.Position=&self:ClampVectorByLength(
     Vector:Vector3f=Linked() {linkedVariable {name=Particles.Position, type=NiagaraPosition}}
     Max Length:NiagaraFloat=&RandomRangeFloat(Minimum=10, Maximum=&RandomRangeFloat(Minimum=10, Maximum=20))
    )) {isSetParameters=true}
   /SetVariables(Particles.PositionMod:NiagaraPosition=&AddVectorToPosition(
     Position=&TransformPosition(Position=$Particles.ParentPosition, Source Space=::World, Destination Space=::Local)
     Vector=Linked() {linkedVariable {name=Particles.Position, type=NiagaraPosition}}
    )) {isSetParameters=true}
   &UpdateMeshOrientation(
    Facing Direction=&SubtractVectorFromPosition(
     Position=$Particles.PositionMod
     Vector=&TransformPosition(Position=$Particles.ParentPosition, Source Space=::World, Destination Space=::Local)
    )
    Facing Interpolation=false
    Orientation Method=::NewEnumerator1
    Override Quat=false
    Stabilization Mode:ENiagara_OrientVectorCount=::NewEnumerator0
    TransformFacing=false
   )
   &CalculateAccurateVelocity()
  ]
  rendererClasses [&NiagaraSpriteRendererProperties, &NiagaraMeshRendererProperties]
  renderers [
   {rendererClass=&NiagaraSpriteRendererProperties}
   {rendererClass=&NiagaraMeshRendererProperties, rendererIndex=1}
   {rendererClass=&NiagaraMeshRendererProperties, rendererIndex=2}
   {rendererClass=&NiagaraSpriteRendererProperties, rendererIndex=3}
  ]
 }
 SatellitesTrail {
  emitterSpawnScript [/SetVariables(Emitter.PebbleReader=DataInterface() {propertyValues {EmitterBinding {bindingMode=Other, emitterName=RockSatellites}}}) {isSetParameters=true}]
  emitterUpdateScript [
   &EmitterState()
   &SpawnRate(SpawnRate=100, Use Spawn Probability=false)
  ]
  particleSpawnScript [
   &PartitionParticles(Number Of Partitions=8, Write Partition to Ribbon ID=::NewEnumerator0)
   &self:ReadParentInfo_Trail(
    Parent Index:NiagaraInt32=$Particles.PartitionParticles.Partition
    Particle Attribute Reader:NiagaraDataInterfaceParticleRead=$Emitter.PebbleReader
   ) {module=ReadParentInfo_Trail}
   &InitializeParticle(
    Color=&SelectLinearColorFromArray(
     Array Sampling Mode=::NewEnumerator0
     Direct Array Index=?
     Direct Array Mode=::NewEnumerator1
     Color Selection Array=DataInterface() {propertyValues {ColorData [<rgba 0.9944 0.539 1 1>, <rgba 1 0.8101 0.519 1>, <rgba 0.519 1 0.6825 1>, <rgba 0.521 0.6841 1 1>]}}
     Random Seed=0
    )
    Lifetime=0.5
    Position=$Particles.ParentPosition
    Position Mode=::NewEnumerator1
    Ribbon Width=&SelectFloatFromArray(Direct Array Index=$Output.ReadParentMesh.ParentMesh, Float Selection Array=DataInterface() {propertyValues {FloatData [6,9,12]}})
    Ribbon Width Mode=::NewEnumerator1
    Sprite Size Mode=::NewEnumerator3
    UsePositionOffset=false
    Write Material Random=true
   )
  ]
  particleUpdateScript [
   &ParticleState()
   &CurlNoiseForce(Bias Noise=false, Noise Strength=20, Pan Noise=false)
   &self:UpdateParentInfo_Trail(Parent Attribute Reader:NiagaraDataInterfaceParticleRead=$Emitter.PebbleReader, Parent ID:NiagaraID=$Particles.ParentID, Push Speed=20) {module=UpdateParentInfo_Trail}
   &AccelerationForce(Acceleration=&Add_Vector(
     A=&Multiply_VectorByFloat(Vector=$Particles.GrandparentVelocity, Float=-0.4)
     B=&Add_Vector(A=<0 0 75>, B=$Particles.Push)
    ))
   &ScaleColor(
    Scale Alpha=&FloatFromCurve(
     FloatCurve=DataInterface() {propertyValues {Curve {keys [
         {interpMode=RCIM_Cubic, leaveTangentWeight=0.2206, tangentMode=RCTM_User, time=0, value=0}
         {arriveTangentWeight=0.1067, interpMode=RCIM_Cubic, leaveTangentWeight=0.3202, tangentMode=RCTM_User, time=0.2, value=1}
         {interpMode=RCIM_Cubic, tangentMode=RCTM_Break, time=0.5, value=1}
         {arriveTangent=-0.1301, tangentMode=RCTM_User, time=1, value=0}
        ]}}}
     Scale Curve=0.35
    )
    Scale RGB=&MakeVectorFromLinearColorRGB(Color=&LerpLinearColors(
      StartColor=&SelectLinearColorFromArray(
       Array Sampling Mode=::NewEnumerator1
       Direct Array Index=$Particles.PartitionParticles.Partition
       Direct Array Mode=::NewEnumerator1
       Color Selection Array=DataInterface() {propertyValues {ColorData [
          <rgba 0.9916 0.308 1 1>
          <rgba 1 0.715 0.278 1>
          <rgba 0.278 1 0.5234 1>
          <rgba 0.282 0.5265 1 1>
          <rgba 1 0.715 0.278 1>
          <rgba 0.278 1 0.5234 1>
          <rgba 0.282 0.5265 1 1>
          <rgba 0.9916 0.308 1 1>
         ]}}
       Random Seed=?
      )
      EndColor=<rgba 0.3069 0.0922 1 1>
      LerpFactor=&FloatFromCurve(FloatCurve=DataInterface() {propertyValues {Curve {
          defaultValue=1
          keys [
           {arriveTangent=1.5484, interpMode=RCIM_Cubic, leaveTangent=1.5484, tangentMode=RCTM_User, time=0, value=0}
           {tangentMode=RCTM_User, time=1, value=1}
          ]
         }}})
     ))
    ScaleA=true
    ScaleRGB=true
   )
   &SolveForcesAndVelocity(Clamp Velocity=false, Limit Acceleration=false)
   &DynamicMaterialParameters(
    Index 0 Param 1=&FloatFromCurve(FloatCurve=DataInterface() {propertyValues {Curve {keys [<at 0.1 1>,<at 1 0.25>]}}})
    Index 0 Param 2=&FloatFromCurve(FloatCurve=DataInterface() {propertyValues {Curve {keys [<at 0 0 RCIM_Cubic>,<at 0.1 1 RCIM_Cubic>,<at 0.5 0 RCIM_Cubic>]}}})
    Index 0 Param 3=-1
    Index 0 Param 4=1
    Module.Param0WriteEnabled=true
    Module.Param1WriteEnabled=true
    Module.Param2WriteEnabled=true
    Module.Param3WriteEnabled=true
   )
   /SetVariables(Particles.RibbonU1Override:NiagaraFloat=&Add_Float(
     A=&Lerp_Float(A=0, B=0.5, Alpha=$Particles.NormalizedAge)
     B=&Multiply_FloatByInt(
      Float=0.05
      Integer=&self:GetIntFromID(Niagara ID:NiagaraID=$Particles.ParentID)
     )
    )) {isSetParameters=true}
   /SetVariables(Particles.PositionMod:NiagaraPosition=&AddVectorToPosition(
     Position=$Particles.GrandparentPosition
     Vector=&MakeCustomVectorFromBool(
      Bool=&SetBoolByFloatComparison(A=$Particles.Age, B=0.08, Comparison Type=::NewEnumerator6)
      True Vector=Linked() {linkedVariable {name=Particles.ParentPosition, type=NiagaraPosition}}
      False Vector=Linked() {linkedVariable {name=Particles.Position, type=NiagaraPosition}}
     )
    )) {isSetParameters=true}
   &Color(Color=&ColorFromCurve(
     ColorCurve=DataInterface() {propertyValues {
       AlphaCurve {keys [<at 0 1 RCIM_Cubic>]}
       BlueCurve {keys [<at 0 0.0851>, {interpMode=RCIM_Cubic, leaveTangent=0.9149, time=0.6272, value=1}, <at 1 1 RCIM_Cubic>]}
       GreenCurve {keys [<at 0 0>, {interpMode=RCIM_Cubic, leaveTangent=0.0761, time=0.6272, value=1}, <at 1 0.0761 RCIM_Cubic>]}
       RedCurve {keys [<at 0 1>, {interpMode=RCIM_Cubic, leaveTangent=-0.998, time=0.6272, value=1}, <at 1 0.002 RCIM_Cubic>]}
      }}
     CurveIndex=&Lerp_Float(
      A=$Particles.NormalizedAge
      B=&Lerp_Float(
       A=$Particles.NormalizedAge
       B=1
       Alpha=&ClampFloat(Float=&Add_Float(
         A=-2
         B=&Multiply_Float(A=0.1, B=&DistanceBetweenPositions(A=$Particles.ParentPosition, B=$Particles.Position))
        ))
      )
      Alpha=0.5
     )
    )) {enabled=false}
  ]
  rendererClasses [&NiagaraSpriteRendererProperties, &NiagaraRibbonRendererProperties]
  renderers [
   {rendererClass=&NiagaraSpriteRendererProperties}
   {rendererClass=&NiagaraRibbonRendererProperties, rendererIndex=1}
  ]
 }
 Stars {
  emitterUpdateScript [
   &EmitterState()
   &SpawnRate(
    SpawnRate=&RandomRangeFloat(Minimum=20, Maximum=45, Evaluation Type=::NewEnumerator1)
    Use Spawn Probability=false
   )
  ]
  particleSpawnScript [
   /SetVariables(Transient.PickBone:NiagaraInt32=&RandomRangeInt(Minimum=1, Maximum=5)) {isSetParameters=true}
   /SetVariables(Transient.CompanionBone:NiagaraInt32=&SelectIntFromArray(Direct Array Index=$Transient.PickBone, Int Selection Array=DataInterface() {propertyValues {IntData [0,2,7,7,7,4]}})) {isSetParameters=true}
   &SkeletalMeshLocation(Bone / Socket Index=$Transient.PickBone, Bone Sampling Mode=::NewEnumerator4, MaskSpawnGroup=false, Skeletal Mesh=$System.Skeletal Mesh)
   &SkeletalMeshLocation(Bone / Socket Index=$Transient.CompanionBone, Bone Sampling Mode=::NewEnumerator4, MaskSpawnGroup=false, Skeletal Mesh=$System.Skeletal Mesh) {module=SkeletalMeshLocation001}
   &InitializeParticle(
    Color=&SelectLinearColorFromArray(
     Array Sampling Mode=::NewEnumerator0
     Direct Array Index=?
     Color Selection Array=DataInterface() {propertyValues {ColorData [<rgba 2.9663 0.2348 3 1>, <rgba 3 1.8589 0.1096 1>, <rgba 0.1096 3 1.092 1>, <rgba 0.1304 1.1075 3 1>]}}
     Random Seed=0
    )
    Lifetime Max=1.5
    Lifetime Min=0.3
    Lifetime Mode=::NewEnumerator1
    Mesh Scale Mode=::NewEnumerator3
    Mesh Uniform Scale=&RandomRangeFloat(Minimum=0.05, Maximum=0.25)
    Position=&Lerp_Vector(
     A=Linked() {linkedVariable {name=Particles.SkeletalMeshLocation.SampledPosition, type=NiagaraPosition}}
     B=Linked() {linkedVariable {name=Particles.SkeletalMeshLocation001.SampledPosition, type=NiagaraPosition}}
     Alpha=&RandomRangeFloat(Minimum=0, Maximum=1)
    )
    Position Mode=::NewEnumerator1
    Sprite Size Mode=::NewEnumerator1
    Uniform Sprite Size Max=1.5
    Uniform Sprite Size Min=4
    UsePositionOffset=false
    Write Material Random=true
   )
   &AddVelocity(Velocity=&RandomRangeVector(Minimum=<-3 -50 15>, Maximum=<3 3 75>))
  ]
  particleUpdateScript [
   &CameraOffset(Camera Offset Amount=-40)
   &ParticleState()
   &ScaleSpriteSize(Uniform Curve Sprite Scale=DataInterface() {propertyValues {Curve {keys [
        {interpMode=RCIM_Cubic, tangentMode=RCTM_User, time=0, value=0.8}
        {interpMode=RCIM_Cubic, tangentMode=RCTM_User, time=0.5, value=1.5}
        {tangentMode=RCTM_User, time=1, value=0}
       ]}}})
   &Drag(Rotational Drag=0.1, Use Linear Drag=true, Use Rotational Drag=true)
   &SolveForcesAndVelocity(Clamp Velocity=false, Limit Acceleration=false)
  ]
  rendererClasses [&NiagaraSpriteRendererProperties]
  renderers [{rendererClass=&NiagaraSpriteRendererProperties}]
 }
}
summary {
 systemName=NS_Aura_Magicstone
 emitters [
  {emitterName=GlowDecal, rendererClasses [&NiagaraSpriteRendererProperties]}
  {emitterName=PerpetualRocks, rendererClasses [&NiagaraMeshRendererProperties, &NiagaraSpriteRendererProperties]}
  {emitterName=Flames, rendererClasses [&NiagaraSpriteRendererProperties]}
  {
   bEnabled=false
   emitterName=Debug
   rendererClasses [&NiagaraSpriteRendererProperties, &NiagaraMeshRendererProperties]
  }
  {emitterName=Dust, rendererClasses [&NiagaraSpriteRendererProperties]}
  {emitterName=RockSatellites, rendererClasses [&NiagaraSpriteRendererProperties, &NiagaraMeshRendererProperties]}
  {emitterName=Stars, rendererClasses [&NiagaraSpriteRendererProperties]}
  {
   emitterName=SatellitesTrail
   rendererClasses [&NiagaraSpriteRendererProperties, &NiagaraRibbonRendererProperties]
  }
 ]
 userVariables [
  User.IsFrontEnd {type=NiagaraBool}
  User.IsLargeBody {type=NiagaraBool}
 ]
}
